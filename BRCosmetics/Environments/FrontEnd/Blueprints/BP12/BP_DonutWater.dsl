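; Front-end BP12 environment water Blueprint (DonutWater) that manages drip audio and a submerge sound mix for its lifetime.
; On BeginPlay it checks its hidden flag, either stopping the drips and popping the mix, or fading the drips in and pushing the mix.
; EndPlay and Destroyed each fade the drips out and pop the mix.
; Overlap, Tick and construction script are empty.
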(event EventDestroyed
  (bind _water_drips_audio (Variables|Sound|GetWater_Drips_Audio)
    (:then
      (Audio|Components|Audio|FadeOut _water_drips_audio 0.25 0.0 "Linear" :!class /Script/Engine.AudioComponent))
    (:else
      _))
  (bind _submerge_mix (Variables|Sound|GetSubmerge%20Mix)
    (:then
      (Audio|PopSoundMixModifier _submerge_mix))
    (:else)))

(event EventEndPlay
  (params (EndPlayReason :type EEndPlayReason%20Enum :default "Destroyed"))
  (bind _water_drips_audio (Variables|Sound|GetWater_Drips_Audio)
    (:then
      (Audio|Components|Audio|FadeOut _water_drips_audio 0.25 0.0 "Linear" :!class /Script/Engine.AudioComponent))
    (:else
      _))
  (bind _submerge_mix (Variables|Sound|GetSubmerge%20Mix)
    (:then
      (Audio|PopSoundMixModifier _submerge_mix))
    (:else)))

(event EventBeginPlay
  (bind _water_drips_audio (Variables|Sound|GetWater_Drips_Audio))
  (if (Variables|Rendering|GetbHidden)
    (Audio|Components|Audio|Stop _water_drips_audio)
    (bind _submerge_mix (Variables|Sound|GetSubmerge%20Mix)
      (:then
        (Audio|PopSoundMixModifier _submerge_mix))
      (:else))
    (else
      (Audio|Components|Audio|FadeIn _water_drips_audio 0.25 1.0 0.0 "Linear" :!class /Script/Engine.AudioComponent)
      (bind _submerge_mix_2 (Variables|Sound|GetSubmerge%20Mix)
        (:then
          (Audio|PushSoundMixModifier _submerge_mix_2))
        (:else)))))

(event Collision|EventActorBeginOverlap
  (params (OtherActor :type Actor%20Object%20Reference)))

(event EventTick
  (params (DeltaSeconds :type Float%20%28single-precision%29)))

(fn ConstructionScript)
