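; Front-end elevator transition actor that fakes doors opening and closing and upward or downward travel, with matching 2D UI sounds.
; Custom events play or reverse timelines driving "Open" and light-pan scalars on the first door material; instant variants set "Open" directly.
; Move timelines fade the motion loop into a stop cue on finish.
; Construction creates two location-named door MIDs and stores the camera transform, which a debug event also prints.
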(event EventBeginPlay)

(event Collision|EventActorBeginOverlap
  (params (OtherActor :type Actor%20Object%20Reference)))

(event EventTick
  (params (DeltaSeconds :type Float%20%28single-precision%29)))

(event Custom|Output%20Camera%20Position
  (bind _camera (Variables|Default|GetCamera))
  (bind _returnvalue (Transformation|GetWorldTransform _camera))
  (Development|PrintString (Utilities|String|ToString(Transform) _returnvalue) true true "(R=0.000000,G=0.660000,B=1.000000,A=1.000000)" 2e+01)
  (Variables|Default|SetTransform%20for%20Copy _returnvalue))

(event Custom|Open%20Doors
  (Audio|PlaySound2D "/BRCosmetics/Sounds/Other/Fort_UI_Sounds/S12_Frontend/UI_Elevator_Door_Open_Cue.UI_Elevator_Door_Open_Cue")
  (bind (_direction _normalizedprogress) (Timeline|Timeline_0 Play)
    (:Update
      (Rendering|Material|SetScalarParameterValue (Variables|Default|GetMID_Doors) "Open" _normalizedprogress :!class /Script/Engine.MaterialInstanceDynamic))))

(event Custom|CloseDoors
  (Audio|PlaySound2D "/BRCosmetics/Sounds/Other/Fort_UI_Sounds/S12_Frontend/UI_Elevator_Door_Close_Cue.UI_Elevator_Door_Close_Cue")
  (Timeline|Timeline_0 Reverse))

(event Custom|StartMoveUp
  (bind _returnvalue (Audio|SpawnSound2D "/BRCosmetics/Sounds/Other/Fort_UI_Sounds/S12_Frontend/UI_Elevator_InMotion_Cue.UI_Elevator_InMotion_Cue"))
  (Variables|Default|SetElevator%20MoveUp%20Audio _returnvalue)
  (bind (_direction _normalizedprogress) (Timeline|Timeline_1 Play)
    (:Update
      (Rendering|Material|SetScalarParameterValue (Variables|Default|GetMID_Doors) "EnableLightpanUp" _normalizedprogress :!class /Script/Engine.MaterialInstanceDynamic))
    (:Finished
      (Audio|Components|Audio|FadeOut (Variables|Default|GetElevator%20MoveUp%20Audio) 0.25 0.0 "Linear" :!class /Script/Engine.AudioComponent)
      (Audio|PlaySound2D "/BRCosmetics/Sounds/Other/Fort_UI_Sounds/S12_Frontend/UI_Elevator_Stop_Cue.UI_Elevator_Stop_Cue"))))

(event Custom|StopMoveUp
  (Timeline|Timeline_1 Reverse))

(event Custom|StartMoveDown
  (bind _returnvalue (Audio|SpawnSound2D "/BRCosmetics/Sounds/Other/Fort_UI_Sounds/S12_Frontend/UI_Elevator_InMotion_Cue.UI_Elevator_InMotion_Cue"))
  (Variables|Default|SetElevator%20MoveDown%20Audio _returnvalue)
  (bind (_direction _normalizedprogress) (Timeline|Timeline_2 Play)
    (:Update
      (Rendering|Material|SetScalarParameterValue (Variables|Default|GetMID_Doors) "EnableLightpanDown" _normalizedprogress :!class /Script/Engine.MaterialInstanceDynamic))
    (:Finished
      (Audio|Components|Audio|FadeOut (Variables|Default|GetElevator%20MoveDown%20Audio) 0.25 0.0 "Linear" :!class /Script/Engine.AudioComponent)
      (Audio|PlaySound2D "/BRCosmetics/Sounds/Other/Fort_UI_Sounds/S12_Frontend/UI_Elevator_Stop_Cue.UI_Elevator_Stop_Cue"))))

(event Custom|StopMoveDown
  (Timeline|Timeline_2 Reverse))

(event Custom|OpenInstant
  (Rendering|Material|SetScalarParameterValue (Variables|Default|GetMID_Doors) "Open" 1.0 :!class /Script/Engine.MaterialInstanceDynamic))

(event Custom|CloseInstant
  (Rendering|Material|SetScalarParameterValue (Variables|Default|GetMID_Doors) "Open" 0.0 :!class /Script/Engine.MaterialInstanceDynamic))

(fn ConstructionScript
  (bind _self self)
  (bind _returnvalue (Transformation|GetActorLocation _self))
  (bind _x (.x _returnvalue))
  (bind _y (.y _returnvalue))
  (bind _z (.z _returnvalue))
  (bind _returnvalue_2 (Math|Float|float+float _x _y _z))
  (bind _returnvalue_3 (/ _returnvalue_2 3.0))
  (bind _returnvalue_4 (Math|Float|Truncate _returnvalue_3))
  (bind _returnvalue_5 (Utilities|String|ToString(Integer) _returnvalue_4))
  (bind _returnvalue_6 (Utilities|GetDisplayName _self))
  (bind _returnvalue_7 (Utilities|String|Append _returnvalue_6 _returnvalue_5))
  (bind _doors (Variables|Default|GetDoors))
  (Variables|Default|SetTransform%20for%20Copy (Transformation|GetWorldTransform (Variables|Default|GetCamera)))
  (bind _returnvalue_8 (Rendering|Material|CreateDynamicMaterialInstance _doors 0 0 (Utilities|String|StringToName _returnvalue_7) :!class /Script/Engine.PrimitiveComponent))
  (Variables|Default|SetMID_Doors _returnvalue_8)
  (bind _returnvalue_9 (Rendering|Material|CreateDynamicMaterialInstance _doors 1 0 (Utilities|String|StringToName (Utilities|String|Append _returnvalue_7 "_B")) :!class /Script/Engine.PrimitiveComponent))
  (Variables|Default|SetMID_Doors_2 _returnvalue_9))
