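; BP_IronDoor_Transition1 is a front-end Blueprint for an iron door transition whose doors are animated entirely through material parameters.
; The construction script creates a dynamic material instance on the Doors component and caches the camera's world transform.
; Custom events play or reverse timelines driving the Open and up/down light-pan parameters, with instant open and close variants setting Open directly.
; Another event prints and stores the current camera transform, apparently for copying.
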(event EventBeginPlay)

(event Collision|EventActorBeginOverlap
  (params (OtherActor :type Actor%20Object%20Reference)))

(event EventTick
  (params (DeltaSeconds :type Float%20%28single-precision%29)))

(event Custom|Output%20Camera%20Position
  (bind _camera (Variables|Default|GetCamera))
  (bind _returnvalue (Transformation|GetWorldTransform _camera))
  (Development|PrintString (Utilities|String|ToString(Transform) _returnvalue) true true "(R=0.000000,G=0.660000,B=1.000000,A=1.000000)" 2e+01)
  (Variables|Default|SetTransform%20for%20Copy _returnvalue))

(event Custom|Open%20Doors
  (bind (_direction _normalizedprogress) (Timeline|Timeline_0 Play)
    (:Update
      (Rendering|Material|SetScalarParameterValue (Variables|Default|GetMID_Doors) "Open" _normalizedprogress :!class /Script/Engine.MaterialInstanceDynamic))))

(event Custom|CloseDoors
  (Timeline|Timeline_0 Reverse))

(event Custom|StartMoveUp
  (bind (_direction _normalizedprogress) (Timeline|Timeline_1 Play)
    (:Update
      (Rendering|Material|SetScalarParameterValue (Variables|Default|GetMID_Doors) "EnableLightpanUp" _normalizedprogress :!class /Script/Engine.MaterialInstanceDynamic))))

(event Custom|StopMoveUp
  (Timeline|Timeline_1 Reverse))

(event Custom|StartMoveDown
  (bind (_direction _normalizedprogress) (Timeline|Timeline_2 Play)
    (:Update
      (Rendering|Material|SetScalarParameterValue (Variables|Default|GetMID_Doors) "EnableLightpanDown" _normalizedprogress :!class /Script/Engine.MaterialInstanceDynamic))))

(event Custom|StopMoveDown
  (Timeline|Timeline_2 Reverse))

(event Custom|OpenInstant
  (Rendering|Material|SetScalarParameterValue (Variables|Default|GetMID_Doors) "Open" 1.0 :!class /Script/Engine.MaterialInstanceDynamic))

(event Custom|CloseInstant
  (Rendering|Material|SetScalarParameterValue (Variables|Default|GetMID_Doors) "Open" 0.0 :!class /Script/Engine.MaterialInstanceDynamic))

(fn ConstructionScript
  (Variables|Default|SetTransform%20for%20Copy (Transformation|GetWorldTransform (Variables|Default|GetCamera)))
  (bind _returnvalue (Rendering|Material|CreateDynamicMaterialInstance (Variables|Default|GetDoors) 0 0 "None" :!class /Script/Engine.PrimitiveComponent))
  (Variables|Default|SetMID_Doors _returnvalue))
