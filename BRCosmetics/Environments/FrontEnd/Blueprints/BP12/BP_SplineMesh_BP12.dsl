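; Spline mesh actor for the BP12 front-end environment that lays a static mesh along a spline.
; The construction script fits the base component to the first segment and adds one per later segment, copying mesh, lighting channels, shadow and forward axis.
; Segments support auto-smoothed tangents, uniform scaling and an override material with per-segment TimeOffset on dynamic instances.
; BeginPlay gathers audio components and stops or plays them according to the actor's hidden state.
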(event EventBeginPlay
  (Variables|Default|SetAudioComponents (Actor|GetComponentsByClass "/Script/Engine.AudioComponent"))
  (Utilities|FlowControl|Sequence
    (:then_0
      _)
    (:then_1
      (Utilities|FlowControl|Delay 1.0)))
  (for _array_element (Variables|Default|GetAudioComponents)
    (if (Variables|Rendering|GetbHidden)
      (if (Audio|Components|Audio|IsPlaying _array_element)
        (Audio|Components|Audio|Stop _array_element))
      (elif (Audio|Components|Audio|IsPlaying _array_element)
        _
        (else
          (Audio|Components|Audio|Play _array_element 0.0))))))

(fn ConstructionScript
  (bind (_bchannel0 _bchannel1 _bchannel2) (Utilities|Struct|BreakLightingChannels (Variables|/Script/Engine.PrimitiveComponent|GetLightingChannels (Variables|SplineMeshActor|GetSplineMeshComponent))))
  (bind (_location _tangent) (Spline|GetLocationandTangentatSplinePoint (Variables|Default|GetSpline) 1 "Local"))
  (bind (_location_2 _tangent_2) (Spline|GetLocationandTangentatSplinePoint (Variables|Default|GetSpline) 0 "Local"))
  (bind _splinemeshcomponent (Variables|SplineMeshActor|GetSplineMeshComponent))
  (bind _scale (Variables|Default|GetScale))
  (bind _scale_2 (Variables|Default|GetScale))
  (bind _splinemeshcomponent_2 (Variables|SplineMeshActor|GetSplineMeshComponent))
  (bind _spline (Variables|Default|GetSpline))
  (bind _overridematerial (Variables|Default|GetOverrideMaterial))
  (bind _overridematerial_2 (Variables|Default|GetOverrideMaterial))
  (SplineMesh|SetStartandEnd _splinemeshcomponent _location_2 _tangent_2 _location _tangent true :!class /Script/Engine.SplineMeshComponent)
  (Utilities|FlowControl|Sequence
    (:then_0
      (if (Utilities|IsValid _overridematerial_2 :!class /Script/Engine.KismetSystemLibrary)
        (Rendering|Material|SetMaterial _splinemeshcomponent 0 _overridematerial_2)))
    (:then_1
      (if (Variables|Default|GetUseScaling)
        (SplineMesh|SetStartScale _splinemeshcomponent (Math|Vector2D|MakeVector2D _scale _scale) true :!class /Script/Engine.SplineMeshComponent)
        (SplineMesh|SetEndScale _splinemeshcomponent (Math|Vector2D|MakeVector2D _scale _scale) true :!class /Script/Engine.SplineMeshComponent))
      (for _index (range 1 (+ (- (Spline|GetNumberOfSplinePoints _spline) 2) 1))
        (bind _returnvalue (AddComponent|Rendering|AddSplineMeshComponent false))
        (Rendering|Components|SetLightingChannels _returnvalue _bchannel0 _bchannel1 _bchannel2)
        (Components|StaticMesh|SetStaticMesh _returnvalue (Variables|/Script/Engine.StaticMeshComponent|GetStaticMesh _splinemeshcomponent_2))
        (Utilities|FlowControl|Sequence
          (:then_0
            (if (Utilities|IsValid _overridematerial :!class /Script/Engine.KismetSystemLibrary)
              (Rendering|Material|SetMaterial _returnvalue 0 _overridematerial)
              (bind _returnvalue_2 (Rendering|Material|CreateDynamicMaterialInstance _returnvalue 0 _overridematerial "None" :!class /Script/Engine.PrimitiveComponent))
              (Rendering|Material|SetScalarParameterValue _returnvalue_2 "TimeOffset" (* _index (Variables|Default|GetTimeoffset)) :!class /Script/Engine.MaterialInstanceDynamic)))
          (:then_1
            (SplineMesh|SetForwardAxis _returnvalue (SplineMesh|GetForwardAxis _splinemeshcomponent) true)
            (Rendering|SetCastShadow _returnvalue (Variables|/Script/Engine.PrimitiveComponent|GetCastShadow _splinemeshcomponent_2))
            (SplineMesh|SetStartandEnd _returnvalue (Spline|GetLocationandTangentatSplinePoint _spline _index "Local") (Math|Vector|SelectVector (* (Math|Vector|Normalize (.Tangent (Spline|GetLocationandTangentatSplinePoint _spline _index "Local"))) (* (Math|Vector|VectorLength (- (Spline|GetLocationandTangentatSplinePoint _spline (+ _index 1) "Local") (Spline|GetLocationandTangentatSplinePoint _spline _index "Local"))) (Variables|Default|GetTension))) (.Tangent (Spline|GetLocationandTangentatSplinePoint _spline _index "Local")) (Variables|Default|GetAutoSmoothTangent)) (Spline|GetLocationandTangentatSplinePoint _spline (+ _index 1) "Local") (Math|Vector|SelectVector (* (Math|Vector|Normalize (.Tangent (Spline|GetLocationandTangentatSplinePoint _spline (+ _index 1) "Local"))) (* (Math|Vector|VectorLength (- (Spline|GetLocationandTangentatSplinePoint _spline (+ _index 1) "Local") (Spline|GetLocationandTangentatSplinePoint _spline _index "Local"))) (Variables|Default|GetTension))) (.Tangent (Spline|GetLocationandTangentatSplinePoint _spline (+ _index 1) "Local")) (Variables|Default|GetAutoSmoothTangent)) true :!class /Script/Engine.SplineMeshComponent)
            (if (Variables|Default|GetUseScaling)
              (SplineMesh|SetStartScale _returnvalue (Math|Vector2D|MakeVector2D _scale_2 _scale_2) true :!class /Script/Engine.SplineMeshComponent)
              (SplineMesh|SetEndScale _returnvalue (Math|Vector2D|MakeVector2D _scale_2 _scale_2) true :!class /Script/Engine.SplineMeshComponent))))))))
