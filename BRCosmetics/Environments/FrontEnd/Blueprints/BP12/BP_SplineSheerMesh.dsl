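; Front-end environment Blueprint for BP12 that builds a configurable sheared mesh from a single spline mesh component, with no gameplay logic.
; The construction script assigns the static mesh, applies the material if valid, and sets the end position from the Length variable.
; Start and end tangents share that straight-line vector, while Offset_X and Offset_Y shift the end cross-section to produce the shear.
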(event EventBeginPlay)

(event Collision|EventActorBeginOverlap
  (params (OtherActor :type Actor%20Object%20Reference)))

(event EventTick
  (params (DeltaSeconds :type Float%20%28single-precision%29)))

(fn ConstructionScript
  (bind _length (Variables|Default|GetLength))
  (bind _returnvalue (Math|Vector|MakeVector _length))
  (bind _returnvalue_2 (Math|Vector|MakeVector 0.0))
  (bind _returnvalue_3 (- _returnvalue _returnvalue_2))
  (bind _returnvalue_4 (Math|Vector|Normalize _returnvalue_3))
  (bind _returnvalue_5 (Math|Vector|VectorLength _returnvalue_3))
  (bind _returnvalue_6 (* _returnvalue_4 _returnvalue_5))
  (bind _material (Variables|Default|GetMaterial))
  (bind _splinemesh (Variables|Default|GetSplineMesh))
  (bind _splinemesh_2 (Variables|Default|GetSplineMesh))
  (Components|StaticMesh|SetStaticMesh _splinemesh_2 (Variables|Default|GetStaticmesh))
  (Utilities|IsValid _material
    (:Is%20Valid
      (Rendering|Material|SetMaterial _splinemesh_2 0 _material))
    (:Is%20Not%20Valid
      _))
  (SplineMesh|SetEndPosition _splinemesh _returnvalue true)
  (SplineMesh|SetEndTangent (Variables|Default|GetSplineMesh) _returnvalue_6 true :!class /Script/Engine.SplineMeshComponent)
  (SplineMesh|SetStartTangent (Variables|Default|GetSplineMesh) _returnvalue_6 true :!class /Script/Engine.SplineMeshComponent)
  (SplineMesh|SetEndOffset _splinemesh (Math|Vector2D|MakeVector2D (Variables|Default|GetOffset_X) (Variables|Default|GetOffset_Y)) true))
