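; Front-end fridge Blueprint for a hidden "Tina" secret in the BP12 lobby, revealing a Banana smoothie mesh.
; BeginPlay creates a dynamic instance of the animated refrigerator material and stores it.
; TriggerFridgeSecret plays a 2D sound and runs the OpenFridge timeline, which drives the material's CODE_Timestamp parameter to animate the door.
; The same timeline lerps the smoothie mesh's relative location and rotation out of the fridge.
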(event EventBeginPlay
  (bind _returnvalue (Rendering|Material|CreateDynamicMaterialInstance (Variables|Default|GetS_OilRig_Refridgerator_02) 0 "/Game/Environments/Apollo/Sets/OilRig/Props/Materials/MI_OilRig_Refridgerators_Anim.MI_OilRig_Refridgerators_Anim" "None" :!class /Script/Engine.PrimitiveComponent))
  (Variables|Default|SetFridgeMat _returnvalue)
  (Utilities|FlowControl|Delay 1e+01))

(event Collision|EventActorBeginOverlap
  (params (OtherActor :type Actor%20Object%20Reference)))

(event EventTick
  (params (DeltaSeconds :type Float%20%28single-precision%29)))

(event Custom|TriggerFridgeSecret
  (Audio|PlaySound2D "/BRCosmetics/Sounds/Other/Quests/SpyBaseSecrets/Tina_Fridge.Tina_Fridge")
  (bind (_direction _dooropentime _smoothmovetime) (Timeline|OpenFridge PlayFromStart)
    (:Update
      (Rendering|Material|SetScalarParameterValue (Variables|Default|GetFridgeMat) "CODE_Timestamp" _dooropentime :!class /Script/Engine.MaterialInstanceDynamic)
      (Transformation|SetRelativeLocationAndRotation (Variables|Default|GetM_MED_Banana_Smoothie_FrontEndSecret) (Math|Vector|Lerp(Vector) "0, 0,96.000000" "-40.000000,30.000000,96.000000" _smoothmovetime) (Math|Rotator|Lerp(Rotator) "0,-180.000000, 0" "-30.000000,-185.000000, 0" _smoothmovetime) false false))))

(fn ConstructionScript)
